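# Physics asset for the F_MED_Toy_Monkey female medium player body, built around a dangling scarf with tassels.
# The scarf chains a tiny root capsule on spine_05 to two boxes, with five tassel capsules fanning from the last box.
# Scarf and tassel joints have tight swing and twist limits, while the root, spine and arm joints rotate freely.
# Capsules cover only the spine, both thighs and right arm, with the thighs unconstrained.
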PhysicsAsset /BRCosmetics/Characters/Player/Female/Medium/Bodies/F_MED_Toy_Monkey/Meshes/F_MED_Toy_Monkey_Physics
bodies {
 dyn_scarf_1 [{shapeType=Capsule, length=0.505, radius=0.505, shapeName=dyn_scarf_1_capsule, rotation=<rot 0 -90 0>}]
 dyn_scarf_2 [{shapeType=Box, extentX=9.2044, extentY=3.6366, extentZ=10, center=<2 0 0>, rotation=<rot 80 90 90>}]
 dyn_scarf_3 [{shapeType=Box, extentX=9.2044, extentY=3.6366, extentZ=8, center=<2 0 0>, rotation=<rot 80 90 90>}]
 dyn_scarf_tassle_1 [{shapeType=Capsule, length=4, radius=1.5, center=<-0.09 -0.4158 -2.3438>, rotation=<rot 13.1371 -89.7977 4.0505>}]
 dyn_scarf_tassle_2 [{shapeType=Capsule, length=4, radius=1.5, center=<-0.1922 -0.5762 -2.3041>, rotation=<rot 6.3242 -90 8.9128>}]
 dyn_scarf_tassle_3 [{shapeType=Capsule, length=4, radius=1.5, center=<-0.0206 -0.2768 -2.125>, rotation=<rot 0.7536 -89.9999 3.2581>}]
 dyn_scarf_tassle_4 [{shapeType=Capsule, length=4, radius=1.5, center=<0.0792 -0.3604 -2.3604>, rotation=<rot 0.6678 -90 -0.2309>}]
 dyn_scarf_tassle_5 [{shapeType=Capsule, length=4, radius=1.5, center=<-0.0361 -0.1801 -1.9489>, rotation=<rot -1.8454 -90 0.1718>}]
 lowerarm_r [{
   shapeType=Capsule
   length=24.0546
   radius=2.9432
   shapeName=lowerarm_r_capsule
   center=<-12.1891 -0.1167 -4.1386>
   rotation=<rot -0.0002 -95.6079 89.9999>
  }]
 spine_01 [{shapeType=Capsule, length=10.6927, radius=13, shapeName=spine_01_capsule, center=<0 -2.8613 0.0315>}]
 spine_03 [{shapeType=Capsule, length=8.0437, radius=13, shapeName=spine_03_capsule, center=<-2.5412 -2.7442 0.0033>}]
 spine_05 [{shapeType=Capsule, length=7, radius=14.5, shapeName=spine_05_capsule, center=<-8.9624 -0.6029 0>}]
 thigh_twist_01_l [{shapeType=Capsule, length=16.0446, radius=9.9122, center=<6.4724 -1.3017 -0.7622>, rotation=<rot -87.1058 -179.9998 179.9999>}]
 thigh_twist_01_r [{shapeType=Capsule, length=17.5209, radius=9.4533, center=<-6.4215 1.1541 0.8873>, rotation=<rot 85.6551 0 0>}]
 upperarm_r [{
   shapeType=Capsule
   length=20.1976
   radius=3.1048
   shapeName=upperarm_r_capsule
   center=<-14.8314 0.7924 -3.5744>
   rotation=<rot -0.0049 -89.9956 -91.0218>
  }]
}
constraints [
 {bone1Name=dyn_scarf_3, bone2Name=dyn_scarf_2, swing1LimitDegrees=20, swing2LimitDegrees=20, twistLimitDegrees=10}
 {
  bone1Name=spine_03
  bone2Name=spine_01
  swing1LimitDegrees=45
  swing1Motion=Free
  swing2LimitDegrees=45
  swing2Motion=Free
  twistLimitDegrees=45
  twistMotion=Free
 }
 {
  bone1Name=spine_05
  bone2Name=spine_03
  swing1LimitDegrees=45
  swing1Motion=Free
  swing2LimitDegrees=45
  swing2Motion=Free
  twistLimitDegrees=45
  twistMotion=Free
 }
 {bone1Name=dyn_scarf_tassle_1, bone2Name=dyn_scarf_3, swing1LimitDegrees=10, swing2LimitDegrees=10, twistLimitDegrees=10}
 {bone1Name=dyn_scarf_tassle_2, bone2Name=dyn_scarf_3, swing1LimitDegrees=10, swing2LimitDegrees=10, twistLimitDegrees=10}
 {bone1Name=dyn_scarf_tassle_3, bone2Name=dyn_scarf_3, swing1LimitDegrees=10, swing2LimitDegrees=10, twistLimitDegrees=10}
 {bone1Name=dyn_scarf_tassle_4, bone2Name=dyn_scarf_3, swing1LimitDegrees=10, swing2LimitDegrees=10, twistLimitDegrees=10}
 {bone1Name=dyn_scarf_tassle_5, bone2Name=dyn_scarf_3, swing1LimitDegrees=10, swing2LimitDegrees=10, twistLimitDegrees=10}
 {
  bone1Name=upperarm_r
  bone2Name=spine_05
  swing1LimitDegrees=45
  swing1Motion=Free
  swing2LimitDegrees=45
  swing2Motion=Free
  twistLimitDegrees=45
  twistMotion=Free
 }
 {
  bone1Name=lowerarm_r
  bone2Name=upperarm_r
  swing1LimitDegrees=45
  swing1Motion=Free
  swing2LimitDegrees=45
  swing2Motion=Free
  twistLimitDegrees=45
  twistMotion=Free
 }
 {bone1Name=dyn_scarf_2, bone2Name=dyn_scarf_1, swing1LimitDegrees=20, swing2LimitDegrees=20, twistLimitDegrees=10}
 {
  bone1Name=dyn_scarf_1
  bone2Name=spine_05
  swing1LimitDegrees=45
  swing1Motion=Free
  swing2LimitDegrees=45
  swing2Motion=Free
  twistLimitDegrees=45
  twistMotion=Free
 }
]
