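# Post-copy animation Blueprint for the F_MED_Toy_Monkey female body, adding secondary motion and cosmetic fit handling.
# It copies the parent mesh pose, then chains a weighted scarf bone constraint, spring-based AnimDynamics on the scarf, RigidBody physics and a Control Rig.
# On customization load, it hides mesh part 4 when the backbling is a cape, otherwise shows it.
# It also caches mesh part 1 with its dynamic and overlay materials.
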AnimBlueprint /BRCosmetics/Characters/Player/Female/Medium/Bodies/F_MED_Toy_Monkey/Meshes/F_MED_Toy_Monkey_Skeleton_AnimBlueprint
target_skeleton=&/F_MED_Toy_Monkey_Skeleton
graphs [
 AnimGraph {nodes [
   "Transform (Modify) Bone - Bone: dyn_scarf_master"(Alpha=1, Translation="0,1.2,0.5", Rotation="0,0,0", Scale="1,1,1") {
    rotationMode=BMM_Additive
    translationMode=BMM_Additive
    alphaBoolBlend {blendOption=Linear}
    boneToModify {boneName=dyn_scarf_master}
    translation {y=1.2, z=0.5}
   }
   n1 = "Constraint - dyn_scarf_1 (PPP)"!(
    ComponentPose=Local To Component(LocalPose=Copy Pose From Mesh() {bCopyCurves=true, bUseAttachedParent=true})
    Alpha=1
    ConstraintWeights_0=0.6283
    ConstraintWeights_1:float=0.3529
    ConstraintWeights_2:float=0.0187
   ) -> (Pose) {
    alphaBoolBlend {blendOption=Linear}
    boneToModify {boneName=dyn_scarf_1}
    constraintSetup [{targetBone {boneName=spine_04}}, {targetBone {boneName=spine_05}}, {targetBone {boneName=clavicle_r}}]
    constraintWeights [1,1,1]
   }
   n2 = Anim Dynamics - Bone: dyn_scarf_1(ComponentPose=@n1, Alpha=1, ExternalForce="0,0,0") {
    angularDampingOverride=0.9999
    angularSpringConstant=50
    bAngularSpring=true
    bOverrideAngularDamping=true
    bUseGravityOverride=true
    simulationSpace=RootRelative
    alphaBoolBlend {blendOption=Linear}
    boundBone {boneName=dyn_scarf_1}
    gravityOverride {z=-300}
    physicsBodyDefinitions [{
      boundBone {boneName=dyn_scarf_1}
      boxExtents=<10 10 10>
      constraintSetup {angularLimitsMax=<0 0 60>, angularLimitsMin=<0 0 2>, angularTarget=<1 0 0>}
      localJointOffset=<2 0 0>
     }]
    relativeSpaceBone {boneName=spine_05}
   }
   n3 = RigidBody(ComponentPose=@n2) {
    bClampLinearTranslationLimitToRefPose=true
    bOverrideWorldGravity=true
    overridePhysicsAsset=&/F_MED_Toy_Monkey_Physics
    simulationSpace=BaseBoneSpace
    alphaBoolBlend {blendOption=Linear}
    baseBoneRef {boneName=root}
    componentLinearAccScale=<0.4 0.4 0.4>
    componentLinearVelScale=<0.6 0.6 0.6>
    overrideWorldGravity {z=-980}
   }
   n4 = Control Rig(Source=Component To Local(ComponentPose=@n3)) {
    alphaBoolBlend {blendOption=Linear}
    controlRigAssetReference {blueprintRigClass=&/F_MED_Toy_Monkey_ControlRig.F_MED_Toy_Monkey_ControlRig_C}
   }
   Output Pose(Result=@n4) {layerGroup=None}
  ]}
 EventGraph {nodes [
   n1 = Event OnCharacterCustomizationLoaded()
   > n2 = Branch(Condition=CheckForBackbling().bIsBackblingACape)
   n3 = "Get (a copy)"(Array:Array of Skeletal Mesh Component Budgeted Object References=@n1.CustomMeshParts, Dimension 1=4) -> (Output:Skeletal Mesh Component Budgeted Object Reference)
   > n4 = SetVisibility!(execute, self:Scene Component Object Reference=@n3, bNewVisibility:bool=false, bPropagateToChildren:bool=false) -> (then)
   n5 = Set SkelMesh(
    execute=[@n4.then, @n8.then]
    SkelMesh:Skeletal Mesh Component Budgeted Object Reference="Get (a copy)"(Array=@n1.CustomMeshParts, Dimension 1=1)
   ) -> (Output_Get:Skeletal Mesh Component Budgeted Object Reference)
   > n6 = Get Dynamic Material Instance(SkeletalMeshComponent=@n5)
   > Set Material Instance(Material Instance:Material Instance Dynamic Object Reference=@n6) -> (Output_Get:Material Instance Dynamic Object Reference)
   > n7 = GetOrCreateDynamicOverlayMaterial(MeshComponent=@n5, MaterialName=None)
   > Set Mid Overlay(Mid Overlay:Material Instance Dynamic Object Reference=@n7) -> (Output_Get:Material Instance Dynamic Object Reference)
   n8 = SetVisibility(execute=@n2.else, self=@n3, bNewVisibility=true, bPropagateToChildren=false)
  ]}
]
variables [
 Material Instance {type=MaterialInstanceDynamic}
 Mid Overlay {type=MaterialInstanceDynamic}
 SkelMesh {type=SkeletalMeshComponentBudgeted}
]
