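# Physics asset for the face accessory part of the female medium Toy Monkey character, defining bodies for its dynamic hair bones.
# Capsules cover each hair bone, and a single box on the head bone anchors them.
# Hair bones 2, 4 and 5 attach directly to the head, while 1a leads to 1b and 3a branches to 3b and 3c.
# Tight swing and twist limits keep movement small, with looser swing on some tips.
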PhysicsAsset /BRCosmetics/Characters/Player/Female/Medium/Bodies/F_MED_Toy_Monkey/Meshes/Parts/F_MED_Toy_Monkey_FaceAcc_Physics
bodies {
 dyn_hair_1a [{shapeType=Capsule, length=2.3996, radius=1.7231, center=<-0.796 0 2.2247>, rotation=<rot 19.688 0 0>}]
 dyn_hair_1b [{shapeType=Capsule, length=3.577, radius=2.0561, center=<-1.2025 0.998 1.6411>, rotation=<rot 0 -128.6422 -57.8336>}]
 dyn_hair_2 [{shapeType=Capsule, length=1.9753, radius=1.4444, center=<4.0027 0.3497 1.0468>, rotation=<rot 0 -85.5427 -85.4166>}]
 dyn_hair_3a [{shapeType=Capsule, length=2.327, radius=1.2749, center=<0.8848 1.5345 1.8702>, rotation=<rot -25.3131 0 36.5632>}]
 dyn_hair_3b [{shapeType=Capsule, length=2.4785, radius=1.8868, center=<1.0814 1.853 1.0293>, rotation=<rot 0 154.9149 -57.0324>}]
 dyn_hair_3c [{shapeType=Capsule, length=0.6864, radius=1.3778, center=<0.8767 1.2811 0.5689>, rotation=<rot 57.2607 -54.6926 43.4838>}]
 dyn_hair_4 [{shapeType=Capsule, length=1.2233, radius=1.498, center=<3.6931 0.0963 -0.0979>, rotation=<rot 0 -89.8114 -71.4255>}]
 dyn_hair_5 [{shapeType=Capsule, length=3.1511, radius=0.8926, center=<3.336 -0.0498 0.1821>, rotation=<rot 0 -86.4821 -79.2348>}]
 head [{shapeType=Box, extentX=31.7014, extentY=25.8894, extentZ=23.4384}]
}
constraints [
 {bone1Name=dyn_hair_1b, bone2Name=dyn_hair_1a, swing1LimitDegrees=25, swing2LimitDegrees=25, twistLimitDegrees=8}
 {bone1Name=dyn_hair_1a, bone2Name=head, swing1LimitDegrees=8, swing2LimitDegrees=8, twistLimitDegrees=5}
 {bone1Name=dyn_hair_3a, bone2Name=head, swing1LimitDegrees=8, swing2LimitDegrees=8, twistLimitDegrees=2}
 {bone1Name=dyn_hair_3b, bone2Name=dyn_hair_3a, swing1LimitDegrees=8, swing2LimitDegrees=8, twistLimitDegrees=2}
 {bone1Name=dyn_hair_3c, bone2Name=dyn_hair_3a, swing1LimitDegrees=20, swing2LimitDegrees=20, twistLimitDegrees=5}
 {bone1Name=dyn_hair_5, bone2Name=head, swing1LimitDegrees=10, swing2LimitDegrees=10, twistLimitDegrees=10}
 {bone1Name=dyn_hair_4, bone2Name=head, swing1LimitDegrees=10, swing2LimitDegrees=20, twistLimitDegrees=10}
 {bone1Name=dyn_hair_2, bone2Name=head, swing1LimitDegrees=10, swing2LimitDegrees=10, twistLimitDegrees=2}
]
