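# Physics asset for the face accessory of the female medium Toy Monkey player body, setting up physics for its two ears.
# It has three box bodies: a large one on the head and a small offset one on each of the left and right dynamic ear bones.
# Each ear is constrained to the head with both swing axes free, so the swing angle values stored on the joints do not restrict motion.
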PhysicsAsset /BRCosmetics/Characters/Player/Female/Medium/Bodies/F_MED_Toy_Monkey/Meshes/Parts/F_MED_Toy_Monkey_FaceAcc_Physics1
bodies {
 dyn_ear_l [{shapeType=Box, extentX=6.3109, extentY=2.1924, extentZ=8.0124, center=<3.7398 0.2706 0.3745>}]
 dyn_ear_r [{shapeType=Box, extentX=6.3109, extentY=2.1924, extentZ=8.0124, center=<-4.0126 0.2706 0.3745>}]
 head [{shapeType=Box, extentX=31.7014, extentY=25.8894, extentZ=23.4384}]
}
constraints [
 {bone1Name=dyn_ear_r, bone2Name=head, swing1LimitDegrees=45, swing1Motion=Free, swing2LimitDegrees=45, swing2Motion=Free, twistLimitDegrees=5}
 {bone1Name=dyn_ear_l, bone2Name=head, swing1LimitDegrees=45, swing1Motion=Free, swing2LimitDegrees=65, swing2Motion=Free, twistLimitDegrees=5}
]
