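# Post-process-style Anim Blueprint for the toy monkey face accessory, adding secondary physics motion over the pose copied from the parent mesh.
# Two spine-relative RigidBody passes feed AnimDynamics on dyn_mouth, then a Control Rig borrowed from SageTwig.
# Skydiving feeds altitude-seeded Perlin noise as horizontal force into the first pass, damped under parachute.
# HeadLimit enables the rig except during emotes, front end, or without a pawn owner.
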AnimBlueprint /BRCosmetics/Characters/Player/Female/Medium/Bodies/F_MED_Toy_Monkey/Meshes/Parts/F_MED_Toy_Monkey_FaceAcc_Skeleton_AnimBlueprint
target_skeleton=&/F_MED_Toy_Monkey_FaceAcc_Skeleton
graphs [
 AnimGraph {nodes [
   n1 = RigidBody(
    ComponentPose=Local To Component(LocalPose=Copy Pose From Mesh() {bUseAttachedParent=true})
    ExternalForce=Get Falling Anim Dyn Up Force Vec 3() -> (Falling Anim Dyn Up Force Vec 3:Vector)
   ) {
    bOverrideWorldGravity=true
    overridePhysicsAsset=&/F_MED_Toy_Monkey_FaceAcc_Physics
    simulationSpace=BaseBoneSpace
    alphaBoolBlend {blendOption=Linear}
    baseBoneRef {boneName=spine_05}
    componentAppliedLinearAccClamp=<500 500 500>
    componentLinearAccScale=<0.25 0.25 0.25>
    componentLinearVelScale=<0.8 0.8 0.8>
    overrideWorldGravity {z=-980}
   }
   n2 = RigidBody(ComponentPose=@n1) {
    bOverrideWorldGravity=true
    overridePhysicsAsset=&/F_MED_Toy_Monkey_FaceAcc_Physics1
    simulationSpace=BaseBoneSpace
    alphaBoolBlend {blendOption=Linear}
    baseBoneRef {boneName=spine_05}
    componentLinearAccScale=<0.4 0.4 0.4>
    componentLinearVelScale=<0.4 0.4 0.4>
    overrideWorldGravity {z=-1000}
   }
   n3 = Anim Dynamics - Bone: dyn_mouth(ComponentPose=@n2, Alpha=0.4, ExternalForce="0,0,0") {
    angularDampingOverride=0.97
    angularSpringConstant=30000
    bAngularSpring=true
    bGravityOverrideInSimSpace=true
    bOverrideAngularDamping=true
    bOverrideLinearDamping=true
    gravityScale=0.35
    linearDampingOverride=0.98
    linearSpringConstant=200
    numSolverIterationsPostUpdate=4
    numSolverIterationsPreUpdate=16
    alphaBoolBlend {blendOption=Linear}
    boundBone {boneName=dyn_mouth}
    physicsBodyDefinitions [{
      sphereCollisionRadius=0
      boundBone {boneName=dyn_mouth}
      boxExtents=<15 15 5>
      constraintSetup {angularLimitsMax=<2 0 0>, angularLimitsMin=<-2 0 0>, angularTarget=<0 0 1>}
      localJointOffset=<0 0 1>
     }]
    relativeSpaceBone {boneName=head}
   }
   n4 = Control Rig!(Source=Component To Local(ComponentPose=@n3), bAlphaBoolEnabled:bool=Get HeadLimit() -> (HeadLimit:bool)) -> (Pose) {
    alphaInputType=Bool
    alphaBoolBlend {blendOption=Linear}
    controlRigAssetReference {blueprintRigClass=/BRCosmetics/Characters/Player/Male/Medium/Bodies/M_MED_SageTwig/Meshes/Parts/M_MED_SageTwig_FaceAcc_ControlRig.M_MED_SageTwig_FaceAcc_ControlRig_C}
   }
   Output Pose(Result=@n4) {layerGroup=None}
  ]}
 EventGraph {nodes [
   Event BlueprintUpdateAnimation()
   > n1 = Sequence()
   n2 = Get PawnLastLocation() -> (PawnLastLocation:Vector)
   n3 = SelectFloat(
    A=Make Perlin Noise and Remap!(Value:float=BreakVector(InVec=@n2).Z, RangeOutMin:float=-15000, RangeOutMax:float=15000) -> (ReturnValue:float)
    B=0
    bPickA=Get bIsSkyDiving()
   )
   n4 = SelectFloat(A=float * float(A=@n3, B=0.2), B=@n3, bPickA=Get bIsParachuteOpened())
   > Set Falling Anim Dyn Up Force Vec 3(Falling Anim Dyn Up Force Vec 3:Vector=MakeVector(X=@n4, Y=@n4)) -> (Output_Get:Vector)
   n5 = Is Valid(exec=@n1.then_1, InputObject=TryGetPawnOwner())
   n6 = Set CinePosing(execute=@n5.Is Not Valid, CinePosing:bool=true) -> (Output_Get:bool)
   n7 = Get bIsPlayingEmote()
   n8 = Get bIsInFrontEnd()
   n9 = Get CinePosing() -> (CinePosing:bool)
   Set HeadLimit(
    execute=[@n6.then, @n10.then]
    HeadLimit:bool=AND Boolean!(A=NOT Boolean(A=@n7), B=NOT Boolean(A=@n8), C:bool=NOT Boolean(A=@n9)) -> (ReturnValue)
   ) -> (Output_Get:bool)
   n10 = Set CinePosing(execute=@n5.Is Valid, CinePosing=false)
   TryGetPawnOwner()
  ]}
]
variables [
 CinePosing {type=boolean}
 Falling Anim Dyn Up Force Vec 3 {type=Vector}
 HeadLimit {type=boolean, default=true}
]
